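# Physics asset for the F_MED_Candor_Tech female body cosmetic, setting up capsule bodies and joints for its skirt's secondary motion.
# Eight skirt chains (front/back, A/B, left/right) each have a tiny master root locked to the pelvis, followed by three larger, offset capsules.
# Chain joints use tight swing and twist limits, stiffest on front B.
# Pelvis and thigh capsules represent the body, with both thighs locked to the pelvis.
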PhysicsAsset /BRCosmetics/Characters/Player/Female/Medium/Bodies/F_MED_Candor_Tech/Meshes/F_MED_Candor_Tech_Physics
bodies {
 dyn_skirt_00_bkA_l_MSTR [{shapeType=Capsule, length=0.505, radius=0.505, shapeName=dyn_skirt_00_bkA_l_MSTR_capsule, rotation=<rot 0 -90 0>}]
 dyn_skirt_00_bkA_r_MSTR [{shapeType=Capsule, length=0.505, radius=0.505, shapeName=dyn_skirt_00_bkA_r_MSTR_capsule, rotation=<rot 0 -90 0>}]
 dyn_skirt_00_bkB_l_MSTR [{shapeType=Capsule, length=0.505, radius=0.505, shapeName=dyn_skirt_00_bkB_l_MSTR_capsule, rotation=<rot 0 -90 0>}]
 dyn_skirt_00_bkB_r_MSTR [{shapeType=Capsule, length=0.505, radius=0.505, shapeName=dyn_skirt_00_bkB_r_MSTR_capsule, rotation=<rot 0 -90 0>}]
 dyn_skirt_00_ftA_l_MSTR [{shapeType=Capsule, length=0.505, radius=0.505, shapeName=dyn_skirt_00_ftA_l_MSTR_capsule, rotation=<rot 0 -90 0>}]
 dyn_skirt_00_ftA_r_MSTR [{shapeType=Capsule, length=0.505, radius=0.505, shapeName=dyn_skirt_00_ftA_r_MSTR_capsule, rotation=<rot 0 -90 0>}]
 dyn_skirt_00_ftB_l_MSTR [{shapeType=Capsule, length=0.505, radius=0.505, shapeName=dyn_skirt_00_ftB_l_MSTR_capsule, rotation=<rot 0 -90 0>}]
 dyn_skirt_00_ftB_r_MSTR [{shapeType=Capsule, length=0.505, radius=0.505, shapeName=dyn_skirt_00_ftB_r_MSTR_capsule, rotation=<rot 0 -90 0>}]
 dyn_skirt_01_bkA_l [{shapeType=Capsule, length=16, radius=8, shapeName=dyn_skirt_01_bkA_l_capsule, center=<4 0 7.3>, rotation=<rot 0 0 90>}]
 dyn_skirt_01_bkA_r [{shapeType=Capsule, length=16, radius=8, shapeName=dyn_skirt_01_bkA_r_capsule, center=<4 0 7.5>, rotation=<rot 0 0 90>}]
 dyn_skirt_01_bkB_l [{shapeType=Capsule, length=16, radius=8, shapeName=dyn_skirt_01_bkB_l_capsule, center=<4 0 8.6>, rotation=<rot 0 0 90>}]
 dyn_skirt_01_bkB_r [{shapeType=Capsule, length=16, radius=8, shapeName=dyn_skirt_01_bkB_r_capsule, center=<4 0 8.7>, rotation=<rot 0 0 90>}]
 dyn_skirt_01_ftA_l [{shapeType=Capsule, length=10, radius=5, shapeName=dyn_skirt_01_ftA_l_capsule, center=<4 0 3>, rotation=<rot 0 0 90>}]
 dyn_skirt_01_ftA_r [{shapeType=Capsule, length=10, radius=5, shapeName=dyn_skirt_01_ftA_r_capsule, center=<4 0 3>, rotation=<rot 0 0 90>}]
 dyn_skirt_01_ftB_l [{shapeType=Capsule, length=16, radius=8, shapeName=dyn_skirt_01_ftB_l_capsule, center=<4 0 9.1>, rotation=<rot 0 0 90>}]
 dyn_skirt_01_ftB_r [{shapeType=Capsule, length=16, radius=8, shapeName=dyn_skirt_01_ftB_r_capsule, center=<4 0 8.8>, rotation=<rot 0 0 90>}]
 dyn_skirt_02_bkA_l [{shapeType=Capsule, length=16, radius=8, shapeName=dyn_skirt_02_bkA_l_capsule, center=<5 0 5>, rotation=<rot 0 0 90>}]
 dyn_skirt_02_bkA_r [{shapeType=Capsule, length=16, radius=8, shapeName=dyn_skirt_02_bkA_r_capsule, center=<4 0 5.5>, rotation=<rot 0 0 90>}]
 dyn_skirt_02_bkB_l [{shapeType=Capsule, length=16, radius=8, shapeName=dyn_skirt_02_bkB_l_capsule, center=<4 0 7.9>, rotation=<rot 0 0 90>}]
 dyn_skirt_02_bkB_r [{shapeType=Capsule, length=16, radius=8, shapeName=dyn_skirt_02_bkB_r_capsule, center=<4 0 7.9>, rotation=<rot 0 0 90>}]
 dyn_skirt_02_ftA_l [{shapeType=Capsule, length=16, radius=8, shapeName=dyn_skirt_02_ftA_l_capsule, center=<4 0 4.9>, rotation=<rot 0 0 90>}]
 dyn_skirt_02_ftA_r [{shapeType=Capsule, length=16, radius=8, shapeName=dyn_skirt_02_ftA_r_capsule, center=<4 0 4.7>, rotation=<rot 0 0 90>}]
 dyn_skirt_02_ftB_l [{shapeType=Capsule, length=16, radius=8, shapeName=dyn_skirt_02_ftB_l_capsule, center=<4 0 8.3>, rotation=<rot 0 0 90>}]
 dyn_skirt_02_ftB_r [{shapeType=Capsule, length=16, radius=8, shapeName=dyn_skirt_02_ftB_r_capsule, center=<4 -1.1485 7.8>, rotation=<rot 0 0 90>}]
 dyn_skirt_03_bkA_l [{shapeType=Capsule, length=16, radius=8, shapeName=dyn_skirt_03_bkA_l_capsule, center=<6 0 4.5>, rotation=<rot 0 0 90>}]
 dyn_skirt_03_bkA_r [{shapeType=Capsule, length=16, radius=8, shapeName=dyn_skirt_03_bkA_r_capsule, center=<6 0 4.5>, rotation=<rot 0 0 90>}]
 dyn_skirt_03_bkB_l [{shapeType=Capsule, length=16, radius=8, shapeName=dyn_skirt_03_bkB_l_capsule, center=<6 0 5.9>, rotation=<rot 0 0 90>}]
 dyn_skirt_03_bkB_r [{shapeType=Capsule, length=16, radius=8, shapeName=dyn_skirt_03_bkB_r_capsule, center=<6 0 5.7>, rotation=<rot 0 0 90>}]
 dyn_skirt_03_ftA_l [{shapeType=Capsule, length=16, radius=8, shapeName=dyn_skirt_03_ftA_l_capsule, center=<6 0 3.3>, rotation=<rot 0 0 90>}]
 dyn_skirt_03_ftA_r [{shapeType=Capsule, length=16, radius=8, shapeName=dyn_skirt_03_ftA_r_capsule, center=<6 0 3.2>, rotation=<rot 0 0 90>}]
 dyn_skirt_03_ftB_l [{shapeType=Capsule, length=16, radius=8, shapeName=dyn_skirt_03_ftB_l_capsule, center=<6 0 5.8>, rotation=<rot 0 0 90>}]
 dyn_skirt_03_ftB_r [{shapeType=Capsule, length=16, radius=8, shapeName=dyn_skirt_03_ftB_r_capsule, center=<6 0 6.2>, rotation=<rot 0 0 90>}]
 pelvis [{shapeType=Capsule, length=10, radius=8, shapeName=pelvis_capsule, center=<0.5237 -1.1734 0>}]
 thigh_l [{shapeType=Capsule, length=32, radius=8, shapeName=thigh_l_capsule, center=<-16 -0.044 0.548>, rotation=<rot 90 0 0>}]
 thigh_r [{shapeType=Capsule, length=32, radius=8, shapeName=thigh_l_capsule, center=<16 0.044 -0.548>, rotation=<rot 90 0 0>}]
}
constraints [
 {
  bone1Name=dyn_skirt_00_bkA_l_MSTR
  bone2Name=pelvis
  swing1LimitDegrees=45
  swing1Motion=Locked
  swing2LimitDegrees=45
  swing2Motion=Locked
  twistLimitDegrees=45
  twistMotion=Locked
 }
 {bone1Name=dyn_skirt_03_ftB_r, bone2Name=dyn_skirt_02_ftB_r, swing1LimitDegrees=5, swing2LimitDegrees=5, twistLimitDegrees=2}
 {bone1Name=dyn_skirt_01_bkA_l, bone2Name=dyn_skirt_00_bkA_l_MSTR, swing1LimitDegrees=10, swing2LimitDegrees=10, twistLimitDegrees=5}
 {bone1Name=dyn_skirt_02_bkA_l, bone2Name=dyn_skirt_01_bkA_l, swing1LimitDegrees=10, swing2LimitDegrees=10, twistLimitDegrees=5}
 {bone1Name=dyn_skirt_03_bkA_l, bone2Name=dyn_skirt_02_bkA_l, swing1LimitDegrees=10, swing2LimitDegrees=10, twistLimitDegrees=5}
 {
  bone1Name=dyn_skirt_00_bkA_r_MSTR
  bone2Name=pelvis
  swing1LimitDegrees=45
  swing1Motion=Locked
  swing2LimitDegrees=45
  swing2Motion=Locked
  twistLimitDegrees=45
  twistMotion=Locked
 }
 {bone1Name=dyn_skirt_01_bkA_r, bone2Name=dyn_skirt_00_bkA_r_MSTR, swing1LimitDegrees=10, swing2LimitDegrees=10, twistLimitDegrees=5}
 {bone1Name=dyn_skirt_02_bkA_r, bone2Name=dyn_skirt_01_bkA_r, swing1LimitDegrees=10, swing2LimitDegrees=10, twistLimitDegrees=5}
 {bone1Name=dyn_skirt_03_bkA_r, bone2Name=dyn_skirt_02_bkA_r, swing1LimitDegrees=10, swing2LimitDegrees=10, twistLimitDegrees=5}
 {
  bone1Name=dyn_skirt_00_bkB_l_MSTR
  bone2Name=pelvis
  swing1LimitDegrees=45
  swing1Motion=Locked
  swing2LimitDegrees=45
  swing2Motion=Locked
  twistLimitDegrees=45
  twistMotion=Locked
 }
 {bone1Name=dyn_skirt_01_bkB_l, bone2Name=dyn_skirt_00_bkB_l_MSTR, swing1LimitDegrees=10, swing2LimitDegrees=10, twistLimitDegrees=5}
 {bone1Name=dyn_skirt_02_bkB_l, bone2Name=dyn_skirt_01_bkB_l, swing1LimitDegrees=10, swing2LimitDegrees=10, twistLimitDegrees=5}
 {bone1Name=dyn_skirt_03_bkB_l, bone2Name=dyn_skirt_02_bkB_l, swing1LimitDegrees=10, swing2LimitDegrees=10, twistLimitDegrees=5}
 {
  bone1Name=dyn_skirt_00_bkB_r_MSTR
  bone2Name=pelvis
  swing1LimitDegrees=45
  swing1Motion=Locked
  swing2LimitDegrees=45
  swing2Motion=Locked
  twistLimitDegrees=45
  twistMotion=Locked
 }
 {bone1Name=dyn_skirt_01_bkB_r, bone2Name=dyn_skirt_00_bkB_r_MSTR, swing1LimitDegrees=10, swing2LimitDegrees=10, twistLimitDegrees=5}
 {bone1Name=dyn_skirt_02_bkB_r, bone2Name=dyn_skirt_01_bkB_r, swing1LimitDegrees=10, swing2LimitDegrees=10, twistLimitDegrees=5}
 {bone1Name=dyn_skirt_03_bkB_r, bone2Name=dyn_skirt_02_bkB_r, swing1LimitDegrees=10, swing2LimitDegrees=10, twistLimitDegrees=5}
 {
  bone1Name=dyn_skirt_00_ftA_l_MSTR
  bone2Name=pelvis
  swing1LimitDegrees=45
  swing1Motion=Locked
  swing2LimitDegrees=45
  swing2Motion=Locked
  twistLimitDegrees=45
  twistMotion=Locked
 }
 {bone1Name=dyn_skirt_01_ftA_l, bone2Name=dyn_skirt_00_ftA_l_MSTR, swing1LimitDegrees=15, swing2LimitDegrees=15, twistLimitDegrees=5}
 {bone1Name=dyn_skirt_02_ftA_l, bone2Name=dyn_skirt_01_ftA_l, swing1LimitDegrees=10, swing2LimitDegrees=10, twistLimitDegrees=5}
 {bone1Name=dyn_skirt_03_ftA_l, bone2Name=dyn_skirt_02_ftA_l, swing1LimitDegrees=10, swing2LimitDegrees=10, twistLimitDegrees=5}
 {
  bone1Name=dyn_skirt_00_ftA_r_MSTR
  bone2Name=pelvis
  swing1LimitDegrees=45
  swing1Motion=Locked
  swing2LimitDegrees=45
  swing2Motion=Locked
  twistLimitDegrees=45
  twistMotion=Locked
 }
 {bone1Name=dyn_skirt_01_ftA_r, bone2Name=dyn_skirt_00_ftA_r_MSTR, swing1LimitDegrees=15, swing2LimitDegrees=15, twistLimitDegrees=5}
 {bone1Name=dyn_skirt_02_ftA_r, bone2Name=dyn_skirt_01_ftA_r, swing1LimitDegrees=10, swing2LimitDegrees=10, twistLimitDegrees=5}
 {bone1Name=dyn_skirt_03_ftA_r, bone2Name=dyn_skirt_02_ftA_r, swing1LimitDegrees=10, swing2LimitDegrees=10, twistLimitDegrees=5}
 {
  bone1Name=dyn_skirt_00_ftB_l_MSTR
  bone2Name=pelvis
  swing1LimitDegrees=45
  swing1Motion=Locked
  swing2LimitDegrees=45
  swing2Motion=Locked
  twistLimitDegrees=45
  twistMotion=Locked
 }
 {bone1Name=dyn_skirt_01_ftB_l, bone2Name=dyn_skirt_00_ftB_l_MSTR, swing1LimitDegrees=5, swing2LimitDegrees=5, twistLimitDegrees=2}
 {bone1Name=dyn_skirt_02_ftB_l, bone2Name=dyn_skirt_01_ftB_l, swing1LimitDegrees=5, swing2LimitDegrees=5, twistLimitDegrees=2}
 {bone1Name=dyn_skirt_03_ftB_l, bone2Name=dyn_skirt_02_ftB_l, swing1LimitDegrees=5, swing2LimitDegrees=5, twistLimitDegrees=2}
 {
  bone1Name=dyn_skirt_00_ftB_r_MSTR
  bone2Name=pelvis
  swing1LimitDegrees=45
  swing1Motion=Locked
  swing2LimitDegrees=45
  swing2Motion=Locked
  twistLimitDegrees=45
  twistMotion=Locked
 }
 {bone1Name=dyn_skirt_01_ftB_r, bone2Name=dyn_skirt_00_ftB_r_MSTR, swing1LimitDegrees=5, swing2LimitDegrees=5, twistLimitDegrees=2}
 {bone1Name=dyn_skirt_02_ftB_r, bone2Name=dyn_skirt_01_ftB_r, swing1LimitDegrees=5, swing2LimitDegrees=5, twistLimitDegrees=2}
 {
  bone1Name=thigh_l
  bone2Name=pelvis
  swing1LimitDegrees=45
  swing1Motion=Locked
  swing2LimitDegrees=45
  swing2Motion=Locked
  twistLimitDegrees=45
  twistMotion=Locked
 }
 {
  bone1Name=thigh_r
  bone2Name=pelvis
  swing1LimitDegrees=45
  swing1Motion=Locked
  swing2LimitDegrees=45
  swing2Motion=Locked
  twistLimitDegrees=45
  twistMotion=Locked
 }
]
