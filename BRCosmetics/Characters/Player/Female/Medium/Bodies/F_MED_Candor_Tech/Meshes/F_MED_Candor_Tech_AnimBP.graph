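# Post-process style AnimBP for the F_MED_Candor_Tech female body, adding secondary motion to necklace, straps and skirt.
# It copies the parent mesh pose, runs two Control Rigs, then chains AnimDynamics, pelvis-relative RigidBody simulation, left-front skirt constraints and additive skirt master-bone offsets.
# One right skirt offset's alpha comes from PlayerMovement while crouching, otherwise zero.
# The event graph also sets gravity and crouch variables that the AnimGraph never reads.
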AnimBlueprint /BRCosmetics/Characters/Player/Female/Medium/Bodies/F_MED_Candor_Tech/Meshes/F_MED_Candor_Tech_AnimBP
target_skeleton=&/F_MED_Candor_Tech_Skeleton
graphs [
 AnimGraph {nodes [
   n1 = Control Rig(Source=Copy Pose From Mesh() {bCopyCurves=true, bUseAttachedParent=true}) {
    alphaBoolBlend {blendOption=Linear}
    controlRigAssetReference {blueprintRigClass=&/F_MED_Candor_Tech_DfrmRig_CtrlRig.F_MED_Candor_Tech_DfrmRig_CtrlRig_C}
   }
   n2 = Control Rig(Source=@n1) {
    alphaBoolBlend {blendOption=Linear}
    controlRigAssetReference {blueprintRigClass=&/F_MED_Candor_Tech_CtrlRig.F_MED_Candor_Tech_CtrlRig_C}
   }
   n3 = "Anim Dynamics - Chain: dyn_necklace_01 -> dyn_necklace_02"(ComponentPose=Local To Component(LocalPose=@n2), Alpha=1, ExternalForce="0,0,0") {
    angularDampingOverride=0.99
    angularSpringConstant=20000
    bAngularSpring=true
    bChain=true
    bLinearSpring=true
    bOverrideAngularDamping=true
    bOverrideLinearDamping=true
    bUseGravityOverride=true
    linearDampingOverride=1
    linearSpringConstant=200
    simulationSpace=RootRelative
    alphaBoolBlend {blendOption=Linear}
    boundBone {boneName=dyn_necklace_01}
    chainEnd {boneName=dyn_necklace_02}
    componentAppliedLinearAccClamp=<2995 3000 3000>
    componentLinearAccScale=<0.3 0.3 0.3>
    componentLinearVelScale=<0.3 0.3 0.3>
    gravityOverride {z=-980}
    physicsBodyDefinitions [
     {
      boundBone {boneName=dyn_necklace_01}
      boxExtents=<10 20 20>
      constraintSetup {angularLimitsMax=<1 0 1>, angularLimitsMin=<-1 -15 -1>, angularTarget=<0 0 0>, linearAxesMax=<0 1 0>, linearAxesMin=<0 -1 0>}
      localJointOffset=<4 0 0>
     }
     {
      boundBone {boneName=dyn_necklace_02}
      boxExtents=<10 20 20>
      constraintSetup {angularLimitsMax=<5 0 15>, angularLimitsMin=<-5 -15 -15>, angularTarget=<0 0 0>}
      localJointOffset=<4 0 0>
     }
    ]
   }
   n4 = Anim Dynamics - Bone: dyn_strapA_01_l(ComponentPose=@n3, Alpha=1, ExternalForce="0,0,0") {
    angularDampingOverride=1
    angularSpringConstant=20000
    bAngularSpring=true
    bOverrideAngularDamping=true
    bUseGravityOverride=true
    simulationSpace=RootRelative
    alphaBoolBlend {blendOption=Linear}
    boundBone {boneName=dyn_strapA_01_l}
    componentAppliedLinearAccClamp=<2000 2000 2000>
    componentLinearAccScale=<0.1 0.1 0.2>
    componentLinearVelScale=<0.2 0.2 0.4>
    gravityOverride {z=-980}
    physicsBodyDefinitions [{
      boundBone {boneName=dyn_strapA_01_l}
      boxExtents=<20 40 40>
      constraintSetup {angularLimitsMax=<5 10 0>, angularLimitsMin=<-5 -10 0>, angularTarget=<0 0 0>}
      localJointOffset=<4 0 0>
     }]
   }
   n5 = Anim Dynamics - Bone: dyn_strapB_01_l(ComponentPose=@n4, Alpha=1, ExternalForce="0,0,0") {
    angularDampingOverride=1
    angularSpringConstant=20000
    bAngularSpring=true
    bOverrideAngularDamping=true
    bUseGravityOverride=true
    simulationSpace=RootRelative
    alphaBoolBlend {blendOption=Linear}
    boundBone {boneName=dyn_strapB_01_l}
    componentAppliedLinearAccClamp=<2000 2000 2000>
    componentLinearAccScale=<0.1 0.1 0.2>
    componentLinearVelScale=<0.2 0.2 0.4>
    gravityOverride {z=-980}
    physicsBodyDefinitions [{
      boundBone {boneName=dyn_strapB_01_l}
      boxExtents=<20 40 40>
      constraintSetup {angularLimitsMax=<5 10 0>, angularLimitsMin=<-5 -10 0>, angularTarget=<0 0 0>}
      localJointOffset=<4 0 0>
     }]
   }
   n6 = Anim Dynamics - Bone: dyn_strapC_01_l(ComponentPose=@n5, Alpha=1, ExternalForce="0,0,0") {
    angularDampingOverride=1
    angularSpringConstant=20000
    bAngularSpring=true
    bOverrideAngularDamping=true
    bUseGravityOverride=true
    simulationSpace=RootRelative
    alphaBoolBlend {blendOption=Linear}
    boundBone {boneName=dyn_strapC_01_l}
    componentAppliedLinearAccClamp=<2000 2000 2000>
    componentLinearAccScale=<0.1 0.1 0.2>
    componentLinearVelScale=<0.2 0.2 0.4>
    gravityOverride {z=-980}
    physicsBodyDefinitions [{
      boundBone {boneName=dyn_strapC_01_l}
      boxExtents=<20 40 40>
      constraintSetup {angularLimitsMax=<5 10 0>, angularLimitsMin=<-5 -10 0>, angularTarget=<0 0 0>}
      localJointOffset=<4 0 0>
     }]
   }
   n7 = Anim Dynamics - Bone: dyn_strapD_01_l(ComponentPose=@n6, Alpha=1, ExternalForce="0,0,0") {
    angularDampingOverride=1
    angularSpringConstant=20000
    bAngularSpring=true
    bOverrideAngularDamping=true
    bUseGravityOverride=true
    simulationSpace=RootRelative
    alphaBoolBlend {blendOption=Linear}
    boundBone {boneName=dyn_strapD_01_l}
    componentAppliedLinearAccClamp=<2000 2000 2000>
    componentLinearAccScale=<0.1 0.1 0.2>
    componentLinearVelScale=<0.2 0.2 0.4>
    gravityOverride {z=-980}
    physicsBodyDefinitions [{
      boundBone {boneName=dyn_strapD_01_l}
      boxExtents=<20 40 40>
      constraintSetup {angularLimitsMax=<5 10 0>, angularLimitsMin=<-5 -10 0>, angularTarget=<0 0 0>}
      localJointOffset=<4 0 0>
     }]
   }
   n8 = Anim Dynamics - Bone: dyn_strapA_01_r(ComponentPose=@n7, Alpha=1, ExternalForce="0,0,0") {
    angularDampingOverride=1
    angularSpringConstant=20000
    bAngularSpring=true
    bOverrideAngularDamping=true
    bUseGravityOverride=true
    simulationSpace=RootRelative
    alphaBoolBlend {blendOption=Linear}
    boundBone {boneName=dyn_strapA_01_r}
    componentAppliedLinearAccClamp=<2000 2000 2000>
    componentLinearAccScale=<0.1 0.1 0.2>
    componentLinearVelScale=<0.2 0.2 0.4>
    gravityOverride {z=-980}
    physicsBodyDefinitions [{
      boundBone {boneName=dyn_strapA_01_r}
      boxExtents=<20 40 40>
      constraintSetup {angularLimitsMax=<5 10 0>, angularLimitsMin=<-5 -10 0>, angularTarget=<0 0 0>}
      localJointOffset=<4 0 0>
     }]
   }
   n9 = Anim Dynamics - Bone: dyn_strapB_01_r(ComponentPose=@n8, Alpha=1, ExternalForce="0,0,0") {
    angularDampingOverride=1
    angularSpringConstant=20000
    bAngularSpring=true
    bOverrideAngularDamping=true
    bUseGravityOverride=true
    simulationSpace=RootRelative
    alphaBoolBlend {blendOption=Linear}
    boundBone {boneName=dyn_strapB_01_r}
    componentAppliedLinearAccClamp=<2000 2000 2000>
    componentLinearAccScale=<0.1 0.1 0.2>
    componentLinearVelScale=<0.2 0.2 0.4>
    gravityOverride {z=-980}
    physicsBodyDefinitions [{
      boundBone {boneName=dyn_strapB_01_r}
      boxExtents=<20 40 40>
      constraintSetup {angularLimitsMax=<5 10 0>, angularLimitsMin=<-5 -10 0>, angularTarget=<0 0 0>}
      localJointOffset=<4 0 0>
     }]
   }
   n10 = Anim Dynamics - Bone: dyn_strapC_01_r(ComponentPose=@n9, Alpha=1, ExternalForce="0,0,0") {
    angularDampingOverride=1
    angularSpringConstant=20000
    bAngularSpring=true
    bOverrideAngularDamping=true
    bUseGravityOverride=true
    simulationSpace=RootRelative
    alphaBoolBlend {blendOption=Linear}
    boundBone {boneName=dyn_strapC_01_r}
    componentAppliedLinearAccClamp=<2000 2000 2000>
    componentLinearAccScale=<0.1 0.1 0.2>
    componentLinearVelScale=<0.2 0.2 0.4>
    gravityOverride {z=-980}
    physicsBodyDefinitions [{
      boundBone {boneName=dyn_strapC_01_r}
      boxExtents=<20 40 40>
      constraintSetup {angularLimitsMax=<5 10 0>, angularLimitsMin=<-5 -10 0>, angularTarget=<0 0 0>}
      localJointOffset=<4 0 0>
     }]
   }
   n11 = Anim Dynamics - Bone: dyn_strapD_01_r(ComponentPose=@n10, Alpha=1, ExternalForce="0,0,0") {
    angularDampingOverride=1
    angularSpringConstant=20000
    bAngularSpring=true
    bOverrideAngularDamping=true
    bUseGravityOverride=true
    simulationSpace=RootRelative
    alphaBoolBlend {blendOption=Linear}
    boundBone {boneName=dyn_strapD_01_r}
    componentAppliedLinearAccClamp=<2000 2000 2000>
    componentLinearAccScale=<0.1 0.1 0.2>
    componentLinearVelScale=<0.2 0.2 0.4>
    gravityOverride {z=-980}
    physicsBodyDefinitions [{
      boundBone {boneName=dyn_strapD_01_r}
      boxExtents=<20 40 40>
      constraintSetup {angularLimitsMax=<5 10 0>, angularLimitsMin=<-5 -10 0>, angularTarget=<0 0 0>}
      localJointOffset=<4 0 0>
     }]
   }
   n12 = RigidBody!(ComponentPose=@n11, Alpha, OverrideWorldGravity:Vector="0,0,-980", ExternalForce) -> (Pose) {
    bClampLinearTranslationLimitToRefPose=true
    bOverrideWorldGravity=true
    overridePhysicsAsset=&/F_MED_Candor_Tech_Physics
    simulationSpace=BaseBoneSpace
    alphaBoolBlend {blendOption=Linear}
    baseBoneRef {boneName=pelvis}
    componentAppliedLinearAccClamp=<2000 2000 2000>
    componentLinearAccScale=<0.3 0.3 0.45>
    componentLinearVelScale=<0.15 0.15 0.225>
    overrideWorldGravity {z=-980}
   }
   n13 = "Constraint - dyn_skirt_01_ftA_l (P)"!(ComponentPose=@n12, bAlphaBoolEnabled:bool=True, ConstraintWeights_0=1) -> (Pose) {
    alphaInputType=Bool
    alphaBoolBlend {blendInTime=0.1, blendOption=Linear, blendOutTime=0.1}
    alphaScaleBiasClamp {bInterpResult=true}
    boneToModify {boneName=dyn_skirt_01_ftA_l}
    constraintSetup [{targetBone {boneName=dyn_skirt_00_ftA_l_MSTR}}]
    constraintWeights [1]
   }
   n14 = "Constraint - dyn_skirt_02_ftA_l (P)"!(ComponentPose=@n13, bAlphaBoolEnabled:bool=True, ConstraintWeights_0=1) -> (Pose) {
    alphaInputType=Bool
    alphaBoolBlend {blendInTime=0.1, blendOption=Linear, blendOutTime=0.1}
    alphaScaleBiasClamp {bInterpResult=true}
    boneToModify {boneName=dyn_skirt_02_ftA_l}
    constraintSetup [{targetBone {boneName=dyn_skirt_00_ftA_l_MSTR}}]
    constraintWeights [1]
   }
   n15 = "Constraint - dyn_skirt_03_ftA_l (P)"!(ComponentPose=@n14, bAlphaBoolEnabled:bool=True, ConstraintWeights_0=1) -> (Pose) {
    alphaInputType=Bool
    alphaBoolBlend {blendInTime=0.1, blendOption=Linear, blendOutTime=0.1}
    alphaScaleBiasClamp {bInterpResult=true}
    boneToModify {boneName=dyn_skirt_03_ftA_l}
    constraintSetup [{targetBone {boneName=dyn_skirt_00_ftA_l_MSTR}}]
    constraintWeights [1]
   }
   n16 = "Constraint - dyn_skirt_01_ftB_l (P)"!(ComponentPose=@n15, bAlphaBoolEnabled:bool=True, ConstraintWeights_0=1) -> (Pose) {
    alphaInputType=Bool
    alphaBoolBlend {blendInTime=0.1, blendOption=Linear, blendOutTime=0.1}
    alphaScaleBiasClamp {bInterpResult=true}
    boneToModify {boneName=dyn_skirt_01_ftB_l}
    constraintSetup [{targetBone {boneName=dyn_skirt_00_ftB_l_MSTR}}]
    constraintWeights [1]
   }
   n17 = "Constraint - dyn_skirt_02_ftB_l (P)"!(ComponentPose=@n16, bAlphaBoolEnabled:bool=True, ConstraintWeights_0=1) -> (Pose) {
    alphaInputType=Bool
    alphaBoolBlend {blendInTime=0.1, blendOption=Linear, blendOutTime=0.1}
    alphaScaleBiasClamp {bInterpResult=true}
    boneToModify {boneName=dyn_skirt_02_ftB_l}
    constraintSetup [{targetBone {boneName=dyn_skirt_00_ftB_l_MSTR}}]
    constraintWeights [1]
   }
   n18 = "Constraint - dyn_skirt_03_ftB_l (P)"!(ComponentPose=@n17, bAlphaBoolEnabled:bool=True, ConstraintWeights_0=1) -> (Pose) {
    alphaInputType=Bool
    alphaBoolBlend {blendInTime=0.1, blendOption=Linear, blendOutTime=0.1}
    alphaScaleBiasClamp {bInterpResult=true}
    boneToModify {boneName=dyn_skirt_03_ftB_l}
    constraintSetup [{targetBone {boneName=dyn_skirt_00_ftB_l_MSTR}}]
    constraintWeights [1]
   }
   n19 = "Transform (Modify) Bone - Bone: dyn_skirt_00_ftB_l_MSTR"(ComponentPose=@n18, Alpha=1, Translation="0,0,1", Rotation="0,0,0", Scale="1,1,1") {
    translationMode=BMM_Additive
    translationSpace=BCS_BoneSpace
    alphaBoolBlend {blendOption=Linear}
    boneToModify {boneName=dyn_skirt_00_ftB_l_MSTR}
    translation {z=1}
   }
   n20 = "Transform (Modify) Bone - Bone: dyn_skirt_00_ftB_l_MSTR"(ComponentPose=@n19, Alpha=1, Translation="0,0,1", Rotation="0,0,0", Scale="1,1,1") {
    translationMode=BMM_Additive
    translationSpace=BCS_BoneSpace
    alphaBoolBlend {blendOption=Linear}
    boneToModify {boneName=dyn_skirt_00_ftB_l_MSTR}
    translation {z=1}
   }
   n21 = "Transform (Modify) Bone - Bone: dyn_skirt_00_ftA_r_MSTR"(ComponentPose=@n20, Alpha=1, Translation="0,0,0", Rotation="10,0,0", Scale="1,1,1") {
    rotationMode=BMM_Additive
    rotationSpace=BCS_BoneSpace
    translationMode=BMM_Additive
    translationSpace=BCS_BoneSpace
    alphaBoolBlend {blendOption=Linear}
    boneToModify {boneName=dyn_skirt_00_ftA_r_MSTR}
    rotation {pitch=10}
   }
   n22 = "Transform (Modify) Bone - Bone: dyn_skirt_00_ftB_r_MSTR"(
    ComponentPose=@n21
    Alpha=Get thigh_rot_otr_r() -> (thigh_rot_otr_r:double)
    Translation="0,0,0"
    Rotation="10.000,-0.000,0.000"
    Scale="1,1,1"
   ) {
    rotationMode=BMM_Additive
    rotationSpace=BCS_BoneSpace
    translationMode=BMM_Additive
    translationSpace=BCS_BoneSpace
    alphaBoolBlend {blendOption=Linear}
    alphaScaleBiasClamp {bInterpResult=true}
    boneToModify {boneName=dyn_skirt_00_ftB_r_MSTR}
    rotation=<rot 10 0 0>
   }
   Output Pose(Result=Component To Local(ComponentPose=@n22)) {layerGroup=None}
  ]}
 EventGraph {nodes [
   Event BlueprintUpdateAnimation()
   > n1 = Sequence()
   n2 = Get bIsMoving2D()
   > Set isCrouchMoving(isCrouchMoving:bool=AND Boolean(A=Get bIsCrouching(), B=NOT Boolean(A=@n2))) -> (Output_Get:bool)
   > n3 = Branch(Condition=Get bIsFalling())
   > Set gravityOverride(gravityOverride:Vector="0, 0,245") -> (Output_Get:Vector)
   Set gravityOverride(execute=@n3.else, gravityOverride="0, 0,-980")
   n4 = Branch(execute=@n1.then_1, Condition=Get bIsCrouching())
   > n5 = PlayerMovement(AnimInstance=Self-Reference(), XDotProductCoefficient=1, YDotProductCoefficient=1, ZDotProductCoefficient=1)
   > Set thigh_rot_otr_r(thigh_rot_otr_r:double=@n5) -> (Output_Get:double)
   Set thigh_rot_otr_r(execute=@n4.else, thigh_rot_otr_r=0)
   Event BlueprintInitializeAnimation()
   > Set Skel_Mesh(Skel_Mesh=GetOwnerSkeletalMeshComponent())
   GetAlphaFromReferenceTransformDelta(
    SkeletalMeshComponent=Get Skel_Mesh()
    BoneName=thigh_r
    BaseName=pelvis
    DeltaComponentType=EulerY
    InputRange=(X=20,Y=40)
    TwistAxis="1,0,0"
   )
  ]}
]
variables [
 gravityOverride {type=Vector}
 isCrouchMoving {type=boolean}
 Skel_Mesh {type=SkeletalMeshComponent}
 thigh_rot_otr_r {type=number}
]
