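# Anim blueprint for the F_MED_Candor_Tech face accessory mesh, adding secondary physics motion to earrings, ponytail and bangs.
# It copies pose and curves from the attached parent mesh, then runs per-bone AnimDynamics springs in series to the output.
# Spherical limits on head, neck, upper spine and right clavicle constrain the lower ponytail bones.
# An alternative chain-based ponytail and bangs setup remains unconnected, and the event graph has no logic.
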AnimBlueprint /BRCosmetics/Characters/Player/Female/Medium/Bodies/F_MED_Candor_Tech/Meshes/Parts/F_MED_Candor_Tech_FaceAcc_AnimBP
target_skeleton=&/F_MED_Candor_Tech_FaceAcc_Skeleton
graphs [
 AnimGraph {nodes [
   n1 = Anim Dynamics - Bone: dyn_ponytail_01(Alpha=1, ExternalForce="0,0,0") {
    angularDampingOverride=1
    angularSpringConstant=20000
    bAngularSpring=true
    bOverrideAngularDamping=true
    bUseGravityOverride=true
    simulationSpace=BoneRelative
    alphaBoolBlend {blendOption=Linear}
    boundBone {boneName=dyn_ponytail_01}
    componentAppliedLinearAccClamp=<2000 2000 2000>
    componentLinearAccScale=<0.1 0.1 0.2>
    componentLinearVelScale=<0.2 0.2 0.4>
    gravityOverride {z=-980}
    physicsBodyDefinitions [{
      boundBone {boneName=dyn_ponytail_01}
      boxExtents=<15 30 30>
      constraintSetup {angularLimitsMax=<5 0 15>, angularLimitsMin=<-5 -15 -15>, angularTarget=<0 0 0>}
      localJointOffset=<4 0 0>
     }]
    relativeSpaceBone {boneName=root}
   }
   n2 = "Anim Dynamics - Chain: dyn_ponytail_02 -> dyn_ponytail_05"(ComponentPose=@n1, Alpha=1, ExternalForce="0,0,0") {
    angularDampingOverride=1
    angularSpringConstant=20000
    bAngularSpring=true
    bChain=true
    bOverrideAngularDamping=true
    bUseGravityOverride=true
    simulationSpace=BoneRelative
    alphaBoolBlend {blendOption=Linear}
    boundBone {boneName=dyn_ponytail_02}
    chainEnd {boneName=dyn_ponytail_05}
    componentAppliedLinearAccClamp=<2000 2000 2000>
    componentLinearAccScale=<0.1 0.1 0.2>
    componentLinearVelScale=<0.2 0.2 0.4>
    gravityOverride {z=-980}
    physicsBodyDefinitions [
     {
      boundBone {boneName=dyn_ponytail_02}
      boxExtents=<15 30 30>
      constraintSetup {angularLimitsMax=<45 10 45>, angularLimitsMin=<-45 -45 -45>, angularTarget=<0 0 0>}
      localJointOffset=<4 0 0>
     }
     {
      boundBone {boneName=dyn_ponytail_03}
      boxExtents=<20 40 40>
      constraintSetup {angularLimitsMax=<45 10 45>, angularLimitsMin=<-45 -45 -45>, angularTarget=<0 0 0>}
      localJointOffset=<4 0 0>
     }
     {
      boundBone {boneName=dyn_ponytail_04}
      boxExtents=<20 40 40>
      constraintSetup {angularLimitsMax=<45 10 45>, angularLimitsMin=<-45 -45 -45>, angularTarget=<0 0 0>}
      localJointOffset=<4 0 0>
     }
     {
      boundBone {boneName=dyn_ponytail_05}
      boxExtents=<20 40 40>
      constraintSetup {angularLimitsMax=<45 10 45>, angularLimitsMin=<-45 -45 -45>, angularTarget=<0 0 0>}
      localJointOffset=<4 0 0>
     }
    ]
    relativeSpaceBone {boneName=root}
   }
   "Anim Dynamics - Chain: dyn_bangs_01 -> dyn_bangs_03"(ComponentPose=@n2, Alpha=1, ExternalForce="0,0,0") {
    angularDampingOverride=1
    angularSpringConstant=20000
    bAngularSpring=true
    bChain=true
    bOverrideAngularDamping=true
    bUseGravityOverride=true
    alphaBoolBlend {blendOption=Linear}
    boundBone {boneName=dyn_bangs_01}
    chainEnd {boneName=dyn_bangs_03}
    componentAppliedLinearAccClamp=<2000 2000 2000>
    componentLinearAccScale=<0.1 0.1 0.2>
    componentLinearVelScale=<0.2 0.2 0.4>
    gravityOverride {z=-980}
    physicsBodyDefinitions [
     {
      boundBone {boneName=dyn_bangs_01}
      boxExtents=<12 24 24>
      constraintSetup {angularLimitsMax=<1 0 15>, angularLimitsMin=<-1 -45 -15>, angularTarget=<0 0 0>}
      localJointOffset=<4 0 0>
     }
     {
      boundBone {boneName=dyn_bangs_02}
      boxExtents=<12 24 24>
      constraintSetup {angularLimitsMax=<1 0 45>, angularLimitsMin=<-1 -45 -45>, angularTarget=<0 0 0>}
      localJointOffset=<4 0 0>
     }
     {
      boundBone {boneName=dyn_bangs_03}
      boxExtents=<12 24 24>
      constraintSetup {angularLimitsMax=<1 0 45>, angularLimitsMin=<-1 -45 -45>, angularTarget=<0 0 0>}
      localJointOffset=<4 0 0>
     }
    ]
   }
   n3 = Anim Dynamics - Bone: dyn_earring_01_l(
    ComponentPose=Local To Component(LocalPose=Copy Pose From Mesh() {bCopyCurves=true, bUseAttachedParent=true})
    Alpha=1
    ExternalForce="0,0,0"
   ) {
    angularDampingOverride=1
    angularSpringConstant=20000
    bAngularSpring=true
    bOverrideAngularDamping=true
    bUseGravityOverride=true
    alphaBoolBlend {blendOption=Linear}
    boundBone {boneName=dyn_earring_01_l}
    componentAppliedLinearAccClamp=<2000 2000 2000>
    componentLinearAccScale=<0.1 0.1 0.2>
    componentLinearVelScale=<0.2 0.2 0.4>
    gravityOverride {z=-980}
    physicsBodyDefinitions [{
      boundBone {boneName=dyn_earring_01_l}
      boxExtents=<10 20 20>
      constraintSetup {angularLimitsMax=<30 30 30>, angularLimitsMin=<-30 -30 -30>, angularTarget=<0 0 0>}
      localJointOffset=<4 0 0>
     }]
   }
   n4 = Anim Dynamics - Bone: dyn_earring_01_r(ComponentPose=@n3, Alpha=1, ExternalForce="0,0,0") {
    angularDampingOverride=1
    angularSpringConstant=20000
    bAngularSpring=true
    bOverrideAngularDamping=true
    bUseGravityOverride=true
    alphaBoolBlend {blendOption=Linear}
    boundBone {boneName=dyn_earring_01_r}
    componentAppliedLinearAccClamp=<2000 2000 2000>
    componentLinearAccScale=<0.1 0.1 0.2>
    componentLinearVelScale=<0.2 0.2 0.4>
    gravityOverride {z=-980}
    physicsBodyDefinitions [{
      boundBone {boneName=dyn_earring_01_r}
      boxExtents=<10 20 20>
      constraintSetup {angularLimitsMax=<30 30 30>, angularLimitsMin=<-30 -30 -30>, angularTarget=<0 0 0>}
      localJointOffset=<4 0 0>
     }]
   }
   n5 = Anim Dynamics - Bone: dyn_ponytail_01(ComponentPose=@n4, Alpha=1, ExternalForce="0,0,0") {
    angularDampingOverride=0.9999
    angularSpringConstant=100
    bAngularSpring=true
    bGravityOverrideInSimSpace=true
    bOverrideAngularDamping=true
    simulationSpace=BoneRelative
    alphaBoolBlend {blendOption=Linear}
    boundBone {boneName=dyn_ponytail_01}
    componentAppliedLinearAccClamp=<3500 3500 3500>
    componentLinearAccScale=<1.2 1.2 1.2>
    componentLinearVelScale=<1.2 1.2 1.2>
    physicsBodyDefinitions [{
      sphereCollisionRadius=0
      boundBone {boneName=dyn_ponytail_01}
      boxExtents=<20 10 10>
      constraintSetup {angularLimitsMax=<1 5 15>, angularLimitsMin=<-1 -30 -15>, angularTarget=<0 0 0>}
      localJointOffset=<2 0 0>
     }]
    relativeSpaceBone {boneName=root}
   }
   n6 = Anim Dynamics - Bone: dyn_ponytail_02(ComponentPose=@n5, Alpha=1, ExternalForce="0,0,0") {
    angularDampingOverride=0.9999
    angularSpringConstant=100
    bAngularSpring=true
    bGravityOverrideInSimSpace=true
    bOverrideAngularDamping=true
    bUseSphericalLimits=true
    simulationSpace=BoneRelative
    alphaBoolBlend {blendOption=Linear}
    boundBone {boneName=dyn_ponytail_02}
    componentAppliedLinearAccClamp=<3500 3500 3500>
    componentLinearAccScale=<1.2 1.2 1.2>
    componentLinearVelScale=<1.2 1.2 1.2>
    physicsBodyDefinitions [{
      boundBone {boneName=dyn_ponytail_02}
      boxExtents=<20 40 40>
      constraintSetup {angularLimitsMax=<45 10 45>, angularLimitsMin=<-45 -45 -45>, angularTarget=<0 0 0>}
      localJointOffset=<4 0 0>
     }]
    relativeSpaceBone {boneName=root}
    sphericalLimits [{limitRadius=7, drivingBone {boneName=head}, sphereLocalOffset=<8 0 0>}]
   }
   n7 = Anim Dynamics - Bone: dyn_ponytail_03(ComponentPose=@n6, Alpha=1, ExternalForce="0,0,0") {
    angularDampingOverride=0.9999
    angularSpringConstant=100
    bAngularSpring=true
    bGravityOverrideInSimSpace=true
    bOverrideAngularDamping=true
    bUseSphericalLimits=true
    simulationSpace=BoneRelative
    alphaBoolBlend {blendOption=Linear}
    boundBone {boneName=dyn_ponytail_03}
    componentAppliedLinearAccClamp=<3500 3500 3500>
    componentLinearAccScale=<1.2 1.2 1.2>
    componentLinearVelScale=<1.2 1.2 1.2>
    physicsBodyDefinitions [{
      boundBone {boneName=dyn_ponytail_03}
      boxExtents=<20 40 40>
      constraintSetup {angularLimitsMax=<45 10 45>, angularLimitsMin=<-45 -45 -45>, angularTarget=<0 0 0>}
      localJointOffset=<4 0 0>
     }]
    relativeSpaceBone {boneName=root}
    sphericalLimits [{limitRadius=9, drivingBone {boneName=head}, sphereLocalOffset=<3 0 0>}]
   }
   n8 = Anim Dynamics - Bone: dyn_ponytail_04(ComponentPose=@n7, Alpha=1, ExternalForce="0,0,0") {
    angularDampingOverride=0.9999
    angularSpringConstant=100
    bAngularSpring=true
    bGravityOverrideInSimSpace=true
    bOverrideAngularDamping=true
    bUseSphericalLimits=true
    simulationSpace=BoneRelative
    alphaBoolBlend {blendOption=Linear}
    boundBone {boneName=dyn_ponytail_04}
    componentAppliedLinearAccClamp=<3500 3500 3500>
    componentLinearAccScale=<1.2 1.2 1.2>
    componentLinearVelScale=<1.2 1.2 1.2>
    physicsBodyDefinitions [{
      boundBone {boneName=dyn_ponytail_04}
      boxExtents=<20 40 40>
      constraintSetup {angularLimitsMax=<45 10 45>, angularLimitsMin=<-45 -45 -45>, angularTarget=<0 0 0>}
      localJointOffset=<4 0 0>
     }]
    relativeSpaceBone {boneName=root}
    sphericalLimits [
     {limitRadius=6, drivingBone {boneName=neck_02}, sphereLocalOffset=<2 0 0>}
     {limitRadius=10, drivingBone {boneName=head}, sphereLocalOffset=<0 -1 0>}
    ]
   }
   n9 = Anim Dynamics - Bone: dyn_ponytail_05(ComponentPose=@n8, Alpha=1, ExternalForce="0,0,0") {
    angularDampingOverride=0.9999
    angularSpringConstant=100
    bAngularSpring=true
    bGravityOverrideInSimSpace=true
    bOverrideAngularDamping=true
    bUseSphericalLimits=true
    simulationSpace=BoneRelative
    alphaBoolBlend {blendOption=Linear}
    boundBone {boneName=dyn_ponytail_05}
    componentAppliedLinearAccClamp=<3500 3500 3500>
    componentLinearAccScale=<1.2 1.2 1.2>
    componentLinearVelScale=<1.2 1.2 1.2>
    physicsBodyDefinitions [{
      boundBone {boneName=dyn_ponytail_05}
      boxExtents=<20 40 40>
      constraintSetup {angularLimitsMax=<45 10 45>, angularLimitsMin=<-45 -45 -45>, angularTarget=<0 0 0>}
      localJointOffset=<4 0 0>
     }]
    relativeSpaceBone {boneName=root}
    sphericalLimits [
     {limitRadius=7, drivingBone {boneName=neck_02}, sphereLocalOffset=<0 0 0>}
     {limitRadius=10, drivingBone {boneName=spine_05}, sphereLocalOffset=<3 0 0>}
     {limitRadius=8, drivingBone {boneName=clavicle_r}, sphereLocalOffset=<-10 0 0>}
    ]
   }
   n10 = Anim Dynamics - Bone: dyn_bangs_01(ComponentPose=@n9, Alpha=1, ExternalForce="0,0,0") {
    angularDampingOverride=0.9999
    angularSpringConstant=100
    bAngularSpring=true
    bGravityOverrideInSimSpace=true
    bOverrideAngularDamping=true
    simulationSpace=BoneRelative
    alphaBoolBlend {blendOption=Linear}
    boundBone {boneName=dyn_bangs_01}
    componentAppliedLinearAccClamp=<3500 3500 3500>
    componentLinearAccScale=<1.2 1.2 1.2>
    componentLinearVelScale=<1.2 1.2 1.2>
    physicsBodyDefinitions [{
      boundBone {boneName=dyn_bangs_01}
      boxExtents=<12 24 24>
      constraintSetup {angularLimitsMax=<1 0 15>, angularLimitsMin=<-1 -45 -15>, angularTarget=<0 0 0>}
      localJointOffset=<4 0 0>
     }]
    relativeSpaceBone {boneName=root}
   }
   n11 = Anim Dynamics - Bone: dyn_bangs_02(ComponentPose=@n10, Alpha=1, ExternalForce="0,0,0") {
    angularDampingOverride=0.9999
    angularSpringConstant=100
    bAngularSpring=true
    bGravityOverrideInSimSpace=true
    bOverrideAngularDamping=true
    simulationSpace=BoneRelative
    alphaBoolBlend {blendOption=Linear}
    boundBone {boneName=dyn_bangs_02}
    componentAppliedLinearAccClamp=<3500 3500 3500>
    componentLinearAccScale=<1.2 1.2 1.2>
    componentLinearVelScale=<1.2 1.2 1.2>
    physicsBodyDefinitions [{
      boundBone {boneName=dyn_bangs_02}
      boxExtents=<12 24 24>
      constraintSetup {angularLimitsMax=<1 0 15>, angularLimitsMin=<-1 -45 -15>, angularTarget=<0 0 0>}
      localJointOffset=<4 0 0>
     }]
    relativeSpaceBone {boneName=root}
   }
   n12 = Anim Dynamics - Bone: dyn_bangs_03(ComponentPose=@n11, Alpha=1, ExternalForce="0,0,0") {
    angularDampingOverride=0.9999
    angularSpringConstant=100
    bAngularSpring=true
    bGravityOverrideInSimSpace=true
    bOverrideAngularDamping=true
    simulationSpace=BoneRelative
    alphaBoolBlend {blendOption=Linear}
    boundBone {boneName=dyn_bangs_03}
    componentAppliedLinearAccClamp=<3500 3500 3500>
    componentLinearAccScale=<1.2 1.2 1.2>
    componentLinearVelScale=<1.2 1.2 1.2>
    physicsBodyDefinitions [{
      boundBone {boneName=dyn_bangs_03}
      boxExtents=<12 24 24>
      constraintSetup {angularLimitsMax=<1 0 15>, angularLimitsMin=<-1 -45 -15>, angularTarget=<0 0 0>}
      localJointOffset=<4 0 0>
     }]
    relativeSpaceBone {boneName=root}
   }
   Output Pose(Result=Component To Local(ComponentPose=@n12)) {layerGroup=None}
  ]}
 EventGraph {nodes [Event BlueprintUpdateAnimation()]}
]
variables []
